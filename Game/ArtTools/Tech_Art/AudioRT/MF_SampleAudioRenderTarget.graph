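# Material function that samples the RT_Audio_01 render target to give materials audio-reactive values for music spectrum, stems and amplitudes, some labeled events-only.
# The Time input sets U, with 0 as immediate, and four fixed V positions select rows whose RGB channels feed twelve outputs.
# UVs are remapped to texel centers of a 32-texel grid with clamped sampling.
# The AmpScale multiply, debug time parameter and collection parameter reach no output.
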MaterialFunction /Game/ArtTools/Tech_Art/AudioRT/MF_SampleAudioRenderTarget
graph {
 properties {bExposeToLibrary=true}
 nodes [
  CollectionParameter() {collection=/Game/Athena/Events/AudioAnalysis/MPC_Audio_Reactive_Parameters}
  Constant() {desc=Music Scale to hit 1, r=1.2}
  n1 = FunctionInput() {description=0-1 time 0 being immediate, inputName=Time, inputType=FunctionInput_Scalar}
  n2 = AppendVector(A=@n1, B=Constant() {r=1})
  n3 = Divide() {constA=31, constB=32}
  n4 = TextureSample(UVs=Add(A=Multiply(A=@n2, B=@n3), B=Divide() {constA=0.5, constB=32})) {samplerSource=SSM_Clamp_WorldGroupSettings, samplerType=SAMPLERTYPE_LinearColor, texture=&/RT_Audio_01}
  FunctionOutput(None=@n4.G) {outputName="MusicStem_808(EventsOnly)"}
  n5 = AppendVector(A=@n1, B=Constant())
  n6 = Divide() {constA=31, constB=32}
  n7 = TextureSample(UVs=Add(A=Multiply(A=@n5, B=@n6), B=Divide() {constA=0.5, constB=32})) {samplerSource=SSM_Clamp_WorldGroupSettings, samplerType=SAMPLERTYPE_LinearColor, texture=&RT_Audio_01}
  FunctionOutput(None=@n7.B) {outputName="MusicStem_Kick(EventsOnly)"}
  n8 = AppendVector(A=@n1, B=Constant() {r=0.25})
  n9 = Divide() {constA=31, constB=32}
  n10 = TextureSample(UVs=Add(A=Multiply(A=@n8, B=@n9), B=Divide() {constA=0.5, constB=32})) {samplerSource=SSM_Clamp_WorldGroupSettings, samplerType=SAMPLERTYPE_LinearColor, texture=&RT_Audio_01}
  FunctionOutput(None=@n10.B) {outputName="MusicStem_Snare(EventsOnly)"}
  n11 = AppendVector(A=@n1, B=Constant() {r=0.5})
  n12 = Divide() {constA=31, constB=32}
  n13 = TextureSample(UVs=Add(A=Multiply(A=@n11, B=@n12), B=Divide() {constA=0.5, constB=32})) {samplerSource=SSM_Clamp_WorldGroupSettings, samplerType=SAMPLERTYPE_LinearColor, texture=&RT_Audio_01}
  FunctionOutput(None=@n13.B) {outputName="MusicStem_Vocals(EventsOnly)"}
  FunctionOutput(None=@n4.B) {outputName="Amplitude_VO_A(EventsOnly)"}
  FunctionOutput(None=@n7.R) {outputName=SpectrumMusic_01}
  FunctionOutput(None=@n10.R) {outputName=SpectrumMusic_02}
  FunctionOutput(None=@n13.R) {outputName=SpectrumMusic_03}
  FunctionOutput(None=@n4.R) {outputName=SpectrumMusic_04}
  FunctionOutput(None=@n7.G) {outputName=Amplitude_Music}
  FunctionOutput(None=@n10.G) {outputName=Amplitude_Master}
  FunctionOutput(None=@n13.G) {outputName="Amplitude_AGV(EventsOnly)"}
  n14 = FunctionInput() {bUsePreviewValueAsDefault=true, inputName=AmpScale, inputType=FunctionInput_Scalar, previewValue=<1.2 0 0 1>}
  Multiply(A=@n7.R, B=@n14)
  ScalarParameter() {parameterName=DebugTIme}
 ]
}
