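# Material function for the Cactus Dancer backpack that draws an audio-reactive, pixelated animated star, with color and Emissive outputs.
# Pixel-snapped UVs, scaled by music spectrum, sample a star SDF that is quantized into bands, shifted by time and music amplitude, to index a four-step color lookup.
# A star mask, grid overlay and Skin FX mask shape the color.
# Emissive lights alternating bands with music amplitude, scaled by time of day.
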MaterialFunction /BRCosmetics/Accessories/FORT_Backpacks/Backpack_F_MED_Cactus_Dancer/Materials/StarAnim/MF_F_MED_Cactus_Dancer_Backpack_TechArt_Star_Anim
graph {nodes [
  n1 = FunctionInput(Preview=TextureCoordinate()) {inputName=UVs, inputType=FunctionInput_Vector2}
  n2 = ScalarParameter() {defaultValue=0.6, group=Visor, parameterName=VisorStarUVScale}
  n3 = ScalarParameter() {defaultValue=0.1, group=Visor, parameterName=Vertical Offset}
  n4 = NamedRerouteDeclaration(None=Add(
    A=MaterialFunctionCall(UVs=@n1, Texture Scale=@n2) {materialFunction=&ScaleUVsByCenter}.UVs
    B=MaterialFunctionCall(Y=@n3) {materialFunction=&MakeFloat2}.Result
   )) {name=UVS, nodeColor=<rgba 1 0.7969 0 1>}
  n5 = ScalarParameter() {defaultValue=108, group=Visor, parameterName=PixelCellSize}
  n6 = Divide(A=Floor(None=Multiply(A=@n4, B=@n5)), B=@n5)
  n7 = StaticBoolParameter() {defaultValue=true, parameterName=usePixelatedUVs}
  n8 = Multiply(A=Subtract(A=StaticSwitch(True=@n6, False=@n1, Value=@n7)) {constB=0.5}) {constB=0.5}
  n9 = Constant()
  n10 = MaterialFunctionCall!(useSmoothStep, uvx=@n9) {materialFunction=/Game/Accessories/FORT_Backpacks/Backpack_F_MED_Cactus_Rocker/Materials/SampleAudioMPCHistory}
  n11 = OneMinus(None=Multiply(A=ComponentMask(None=@n10.Interpolated Spectrum) {r=true}) {constB=0.8})
  n12 = TextureSample(UVs=Add(A=Multiply(A=@n8, B=@n11)) {constB=0.5}) {
   samplerSource=SSM_Clamp_WorldGroupSettings
   samplerType=SAMPLERTYPE_LinearColor
   texture=/BRCosmetics/Accessories/FORT_Backpacks/Backpack_F_MED_Cactus_Dancer/Textures/starSDF
  }
  n13 = Multiply(A=ScalarParameter() {defaultValue=8, group=Visor, parameterName=LayerWidth}) {constB=0.5}
  n14 = Multiply(A=Add(A=@n12.R), B=@n13) {constB=1.5}
  n15 = OneMinus(None=@n14)
  n16 = Multiply(A=StaticSwitch(
    True=Time()
    False=ScalarParameter() {defaultValue=21, parameterName=TimeFake}
    Value=StaticBoolParameter() {defaultValue=true, parameterName=useTIme}
   )) {constB=3}
  n17 = Constant() {r=2}
  n18 = Divide(A=Constant() {r=2}, B=@n13)
  n19 = StaticBoolParameter() {parameterName=useOffsetStart}
  n20 = StaticSwitch(
   True=Multiply(A=Ceil(None=@n16), B=StaticSwitch(True=@n17, False=@n18, Value=@n19))
   False=@n16
   Value=StaticBoolParameter() {parameterName=stepAnim}
  )
  n21 = Multiply(A=NamedRerouteDeclaration(None=@n10.Interpolated Amplitude Music) {name=Amp, nodeColor=<rgba 0.1484 1 0 1>}) {constB=2}
  n22 = Floor(None=Add(A=@n15, B=Add(A=@n20, B=@n21)))
  n23 = Constant() {r=4}
  n24 = Fmod(A=@n22, B=@n23)
  n25 = TextureSample(UVs=ComponentMask(None=Divide(A=@n24, B=@n23)) {g=true, r=true}) {
   samplerSource=SSM_Wrap_WorldGroupSettings
   samplerType=SAMPLERTYPE_LinearColor
   texture=/BRCosmetics/Accessories/FORT_Backpacks/Backpack_F_MED_Cactus_Dancer/Textures/CA_Cactus_Dancer
  }
  n26 = Multiply(A=ScalarParameter() {defaultValue=6, parameterName=feather}) {constB=-1}
  n27 = Constant()
  n28 = Subtract(A=@n14, B=ScalarParameter() {defaultValue=8, parameterName=maskSize})
  n29 = NamedRerouteDeclaration(None=OneMinus(None=SmoothStep(Min=@n26, Max=@n27, Value=@n28))) {name=star mask, nodeColor=<rgba 0 0.1797 1 1>}
  n30 = LinearInterpolate(A=Constant3Vector() {constant=<rgba 1 0 1 0>}, B=@n25.RGB, Alpha=@n29)
  n31 = NamedRerouteUsage()
  n32 = Multiply(A=ScalarParameter() {defaultValue=108, group=Visor, parameterName=PixelCellSize}) {constB=0.5}
  n33 = TextureSample(UVs=Multiply(A=Multiply(A=@n31, B=@n32), B=ScalarParameter() {defaultValue=0.0139, parameterName=uvfac}) {constB=50}) {samplerType=SAMPLERTYPE_LinearColor, texture=/BRCosmetics/Accessories/FORT_Backpacks/Backpack_F_MED_Cactus_Dancer/Textures/stargrid}
  n34 = OneMinus(None=Multiply(A=OneMinus(None=@n33.R), B=ScalarParameter() {defaultValue=0.4, parameterName=gridOpacity}))
  n35 = Constant3Vector() {constant=<rgba 0.4635 0 0.3804 0>}
  n36 = Constant() {r=1}
  n37 = TextureObjectParameter() {
   group=Skin FX
   parameterName=SkinFX_Mask
   samplerType=SAMPLERTYPE_Masks
   texture=/BRCosmetics/Accessories/FORT_Backpacks/Backpack_F_MED_Cactus_Dancer/Textures/T_F_MED_Cactus_Dancer_Backpack_FX
  }
  n38 = TextureSample(Tex=@n37) {samplerType=SAMPLERTYPE_Masks}
  FunctionOutput(None=Multiply(A=Multiply(A=@n30, B=@n34), B=LinearInterpolate(A=@n35, B=@n36, Alpha=@n38.G) {constAlpha=1}))
  n39 = Multiply(A=@n25.RGB, B=@n29)
  n40 = NamedRerouteUsage()
  n41 = Multiply(A=Multiply(A=Power(Base=@n40)) {constB=6}, B=Fmod(A=@n22, B=Constant() {r=2}))
  n42 = ScalarParameter() {defaultValue=1, parameterName=EmissiveAudio}
  n43 = ScalarParameter() {defaultValue=0.6, parameterName=EmissiveAudioNight}
  FunctionOutput(None=Multiply(
    A=Multiply(A=@n39, B=@n41)
    B=MaterialFunctionCall(Value_Morning=@n42, Value_Day=@n42, Value_Evening=@n42, Value_Night=@n43) {materialFunction=&Scalar_Time_of_Day}.Result
   )) {outputName=Emissive}
  n44 = Add(A=@n15, B=Constant() {r=7})
  n45 = Constant() {r=8}
  n46 = Divide(A=Fmod(A=Floor(None=@n44), B=@n45), B=Subtract(A=@n45))
  n47 = Multiply(A=OneMinus(None=Saturate(None=@n46))) {constB=0.05}
  MaterialFunctionCall!(AmpScale, Time=@n47) {materialFunction=/Game/ArtTools/Tech_Art/AudioRT/MF_SampleAudioRenderTarget}
  MaterialFunctionCall(Time=@n9) {materialFunction=&MF_SampleAudioRenderTarget}
  MaterialFunctionCall!(useSmoothStep, uvx=@n9) {materialFunction=/BRCosmetics/Accessories/FORT_Backpacks/Backpack_F_MED_Cactus_Rocker/Materials/MF_SampleAudioMPCInterpolated}
  MaterialFunctionCall(uvx=@n47) {materialFunction=&MF_SampleAudioMPCInterpolated}
  n48 = Constant3Vector() {constant=<rgba 1 1 0 0>}
  n49 = ComponentMask(None=MaterialFunctionCall(uvx=@n47) {materialFunction=&SampleAudioMPCHistory}.Interpolated Amplitude Music) {r=true}
  n50 = Power(Base=Saturate(None=Add(A=@n49) {constB=0.1})) {constExponent=7}
  Multiply(A=Multiply(A=Multiply(A=@n48, B=@n50)), B=NamedRerouteUsage())
  OneMinus(None=@n38.G)
 ]}
