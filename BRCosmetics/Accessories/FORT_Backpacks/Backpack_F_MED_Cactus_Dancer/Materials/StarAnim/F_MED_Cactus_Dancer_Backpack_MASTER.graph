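# Master character material for the Cactus Dancer backpack cosmetic, adding a star animation to the standard character shading.
# The shared PreFX function feeds CharacterSkinFX, and a tech-art star animation function sampled on UV1 supplies base color and emissive through material layer blends.
# The UseVisor static switch selects the star-blended or plain result before the shared PostFX and cinematic post functions.
# Shading model comes from the material expression, with material attributes throughout.
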Material /BRCosmetics/Accessories/FORT_Backpacks/Backpack_F_MED_Cactus_Dancer/Materials/StarAnim/F_MED_Cactus_Dancer_Backpack_MASTER
graph {
 properties {
  bAutomaticallySetUsageInEditor=false
  bUseMaterialAttributes=true
  bUsedWithMeshDeformer=true
  bUsedWithMorphTargets=true
  bUsedWithSkeletalMesh=true
  pixelDepthOffsetMode=PDOM_Legacy
  refractionMethod=RM_IndexOfRefraction
  shadingModel=MSM_FromMaterialExpression
  subsurfaceProfile=&SS_HeroSkin_02
  usedShadingModels=Default Lit | Subsurface | Clear Coat | Subsurface Profile | Cloth | Eye
 }
 nodes [
  n1 = MaterialFunctionCall() {materialFunction=&BaseFN_CharacterFunction_PreFX}
  n2 = MaterialFunctionCall(
   PreskinnedPosition=@n1.Pre skinned Local Position
   Gmap=@n1.Pre skinned Local Position
   MaterialAttributes=@n1.Pre skinned Local Position
   AO=@n1.Pre skinned Local Position
  ) {materialFunction=&CharacterSkinFX}
  n3 = MaterialFunctionCall!(UVs=TextureCoordinate() {coordinateIndex=1}) {materialFunction=&/MF_F_MED_Cactus_Dancer_Backpack_TechArt_Star_Anim}
  n4 = MaterialFunctionCall!(Material=@n2.Result, NewBaseColor=@n3.Result, Mask) {materialFunction=/Engine/Functions/MaterialLayerFunctions/MatLayerBlend_OverrideBaseColor}
  n5 = MaterialFunctionCall(
   AO=@n1.AO
   Distance Blend=@n1.AO
   Fresnel=@n1.AO
   Gmap=@n1.AO
   Pre Skinned Local Position=@n1.AO
   Material=StaticSwitch(
    True=MaterialFunctionCall!(Material=@n4.Blended Material, Emissive=@n3.Emissive) {materialFunction=/Engine/Functions/MaterialLayerFunctions/MatLayerBlend_Emissive}.Blended Material
    False=@n2.Result
    Value=StaticBoolParameter() {group=Visor, parameterName=UseVisor}
   )
  ) {materialFunction=&BaseFN_CharacterFunction_PostFX}
  MaterialFunctionCall(In=@n5.Result) {materialFunction=&CINE_MF_CharPostFX}
 ]
}
